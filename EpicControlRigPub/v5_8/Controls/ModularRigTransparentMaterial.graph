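# Translucent, Default Lit material in the Modular Rig controls folder, named as a transparent material for rig controls.
# Its graph holds no expressions or parameters, so nothing drives color or opacity and the material inputs stay at their defaults.
# Automatic setting of usage flags in the editor is turned off.
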Material /EpicControlRigPub/v5_8/Controls/ModularRigTransparentMaterial
graph {expressions {}, properties {bAutomaticallySetUsageInEditor=false, blendMode=BLEND_Translucent, usedShadingModels=Default Lit}}
