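# Unlit translucent material for Control Rig modular rig controls that renders as an X-ray, with depth testing disabled so it shows through occluding geometry.
# A Color vector parameter, green by default, is divided by EyeAdaptation so its brightness is independent of scene auto exposure.
# A scalar Opacity parameter, mostly opaque by default, provides the transparency control.
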Material /EpicControlRigPub/v5_8/Controls/ModularRigXRayMaterial
graph {
 properties {
  bDisableDepthTest=true
  bUsedWithStaticLighting=true
  blendMode=BLEND_Translucent
  pixelDepthOffsetMode=PDOM_Legacy
  shadingModel=MSM_Unlit
  usedShadingModels=Unlit
 }
 nodes [
  Divide(A=VectorParameter() {parameterName=Color, defaultValue=<rgba 0 1 0.0898 0>}.RGB, B=EyeAdaptation().EyeAdaptation)
  ScalarParameter() {defaultValue=0.75, parameterName=Opacity}
 ]
}
